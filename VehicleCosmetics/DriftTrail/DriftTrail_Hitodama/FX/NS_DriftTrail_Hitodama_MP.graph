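# Niagara system for the Hitodama vehicle drift trail cosmetic, layering glowing ribbons with flipbook sprites at skeletal-mesh sockets.
# Ribbon emitters use quality-based spawn rates, inherit source movement and mask per tire, while sprite emitters add looping bursts and curl-noise-driven particles.
# User parameters switch between painted-variant colors and default blue gradients with waveform pulsing, apply offsets and adjust frontend previews.
# Several parallel emitter versions remain disabled.
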NiagaraSystem /VehicleCosmetics/DriftTrail/DriftTrail_Hitodama/FX/NS_DriftTrail_Hitodama_MP
dependencies {
 usedDataInterfaces [
  &NiagaraDataInterfaceCurve
  &NiagaraDataInterfaceSpriteRendererInfo
  &NiagaraDataInterfaceColorCurve
  &NiagaraDataInterfaceVectorCurve
  &NiagaraDataInterfaceArrayFloat3
  &NiagaraDataInterfaceArrayColor
  &NiagaraDataInterfaceSkeletalMesh
 ]
 usedDynamicInputs []
 usedModules [
  &self:SystemSpawnScript.SystemScriptSource.SystemScriptGraph.NiagaraNodeAssignment_0.SetVariables_#1
  &FN_UnpackVehicleFlags
  /Game/Effects/Niagara/Modules/Vehicle/FN_UpdateVehicleAttribs
  &SystemState
  /Game/Effects/Niagara/Modules/Vehicle/FN_VehicleFlags_Cheatcodes
 ]
}
emitters {
 FillTrail {
  bEnabled=false
  emitterUpdateScript [
   &EmitterState(Life Cycle Mode=::NewEnumerator1, Loop Duration=1, UseLoopDelay=false)
   &VC_RateQualityHandler(
    Epic Spawn Rate:NiagaraFloat=&MergeHelper_FloatSwitch(0=45, 1=15, 2=45, 3=?, Index=0, Num=::NewEnumerator1)
    High Spawn Rate:NiagaraFloat=&MergeHelper_FloatSwitch(0=35, 1=15, 2=40, 3=?, Index=0, Num=::NewEnumerator1)
    Low Spawn Rate:NiagaraFloat=&MergeHelper_FloatSwitch(0=32, 1=15, 2=30, 3=?, Index=0, Num=::NewEnumerator1)
    Medium Spawn Rate:NiagaraFloat=&MergeHelper_FloatSwitch(0=31, 1=15, 2=35, 3=?, Index=0, Num=::NewEnumerator1)
    Use Custom Rate?=true
   )
   &VC_SpawnRibbon_v1(Number of Ribbons=2, Spawn Time=-1, SpawnRate=$Emitter.Spawn Rate)
  ]
  particleSpawnScript [
   &VC_PaintedVariantsOverride(
    Black=<rgba 0.03 0.03 0.03 1>
    BurntSienna=<rgba 0.184 0.12 0.0378 0.7217>
    Cobalt=<rgba 0 0.0117 0.6348 1>
    Crimson=<rgba 0.2622 0.0664 0.0184 0.7275>
    ForestGreen=<rgba 0.018 0.2656 0.0305 0.8899>
    Grey=<rgba 0.15 0.15 0.15 1.58>
    Lime=<rgba 0.1838 0.2344 0 1>
    Orange=<rgba 0.35 0.1047 0.035 0.8>
    Pink=<rgba 0.6667 0.0676 0.4488 0.0841>
    Purple=<rgba 0.093 0 0.3594 0.8203>
    Saffron=<rgba 0.25 0.2343 0 1>
    SkyBlue=<rgba 0.1828 0.3003 0.3594 0.4261>
    TitaniumWhite=<rgba 0.3108 0.3108 0.3108 0.5>
   )
   &VC_InitializeRibbon_v1(
    Color=$Particles.Color
    Custom Offset=true
    Enable Skeletal Position=true
    Lifetime=&MergeHelper_FloatSwitch(0=0.9, 1=1.1, 2=0.55, 3=?, Index=0, Num=::NewEnumerator1)
    Local Vector Offset=&AddVectorToPosition(Position=<25 0 -15>, Vector=$User.offset)
    Offset Amount=1
    Ribbon Width=4
   )
   &AddVelocity(Velocity=<-1000 0 0>) {module=AddVelocity001}
  ]
  particleUpdateScript [
   &FN_DisableTireFX(Mask Index Attribute:NiagaraInt32=$Particles.SpawnGroup, Mask Index FL=2, Mask Index FR=3, Mask Index RL=0, Mask Index RR=1)
   &InheritSourceMovement(
    Applied Position Scale=&VectorFromFloat(Value=&FloatFromCurve(
      FloatCurve=DataInterface() {propertyValues {Curve {
         defaultValue=1
         keys [<at 0 1 RCIM_Cubic>, {arriveTangent=-3.2426, interpMode=RCIM_Cubic, leaveTangent=-3.2426, tangentMode=RCTM_User, time=1, value=0}]
        }}}
      Scale Curve=1.1
     ))
    ApplyForce=false
    ApplyPosition=true
    ApplyPositionFalloff=false
   )
   &ParticleState()
   &SolveForcesAndVelocity(Clamp Velocity=false, Limit Acceleration=false)
   &VC_UpdateRibbon_v1(Enable Skeletal Position=true)
   &ScaleRibbonWidth(Ribbon Width Scale=&FloatFromCurve(
     FloatCurve=DataInterface() {propertyValues {Curve {
        defaultValue=0
        keys [<at 0 0.7 RCIM_Cubic>, {arriveTangent=-0.7993, interpMode=RCIM_Cubic, leaveTangent=-0.7993, tangentMode=RCTM_User, time=1, value=0.15}]
       }}}
     Scale Curve=30
    ))
   /SetVariables(
    Particles.BC_Principal:LinearColor=&MakeCustomLinearColorFromBool(
     Bool=$User.bIsPainted
     True Color=&MultiplyLinearColor(
      Linear Color=$Particles.Color
      Scale Factor=&MakeVector4FromVectorAndFloat("Vector (RGB)"=<1 1 1>, "Float (Alpha)"=1.3)
     )
     False Color=/DealershipSharedLibrary/FX/Niagara/DynamicInputs/MultiplyLinearColorByVectorAndFloat(
      Color to Scale:LinearColor=&ColorFromCurve(ColorCurve=DataInterface() {propertyValues {
         AlphaCurve {keys [<at 0 2 RCIM_Cubic>,<at 0.3425 1>]}
         BlueCurve {keys [<at 0 0.842 RCIM_Cubic>,<at 0.7299 0.5972>]}
         GreenCurve {keys [<at 0 0.3521 RCIM_Cubic>,<at 0.7299 0>]}
         RedCurve {keys [<at 0 0.0723 RCIM_Cubic>,<at 0.7299 0.0704>]}
        }})
      RGB Scale:Vector3f=<1 1 1>
      Alpha Scale:NiagaraFloat=2
     )
    )
    "Particles.RibbonWidth_(3)":NiagaraFloat=&FloatFromCurve(
     FloatCurve=DataInterface() {propertyValues {Curve {keys [
         {arriveTangent=-0.5501, interpMode=RCIM_Cubic, leaveTangent=-0.5501, tangentMode=RCTM_User, time=0, value=0.8}
         {arriveTangent=-0.5638, interpMode=RCIM_Cubic, leaveTangent=-0.5638, tangentMode=RCTM_User, time=1, value=0.2}
        ]}}}
     Scale Curve=80
    )
   ) {isSetParameters=true, module=SetVariables_#2}
  ]
  rendererClasses [&NiagaraRibbonRendererProperties]
  renderers [
   {rendererClass=&NiagaraRibbonRendererProperties}
   {rendererClass=&NiagaraRibbonRendererProperties, rendererIndex=1}
  ]
 }
 FillTrail001 {
  emitterUpdateScript [
   &EmitterState(Life Cycle Mode=::NewEnumerator1, Loop Duration=1, UseLoopDelay=false)
   &VC_RateQualityHandler(
    Epic Spawn Rate:NiagaraFloat=&MergeHelper_FloatSwitch(0=45, 1=15, 2=45, 3=?, Index=0, Num=::NewEnumerator1)
    High Spawn Rate:NiagaraFloat=&MergeHelper_FloatSwitch(0=35, 1=15, 2=40, 3=?, Index=0, Num=::NewEnumerator1)
    Low Spawn Rate:NiagaraFloat=&MergeHelper_FloatSwitch(0=32, 1=15, 2=30, 3=?, Index=0, Num=::NewEnumerator1)
    Medium Spawn Rate:NiagaraFloat=&MergeHelper_FloatSwitch(0=31, 1=15, 2=35, 3=?, Index=0, Num=::NewEnumerator1)
    Use Custom Rate?=true
   )
   &VC_SpawnRibbon_v1(Number of Ribbons=2, Spawn Time=-1, SpawnRate=$Emitter.Spawn Rate)
  ]
  particleSpawnScript [
   &VC_PaintedVariantsOverride(
    Black=<rgba 0.1 0.1 0.1 1>
    BurntSienna=<rgba 0.5 0.2229 0.025 0.75>
    Cobalt=<rgba 0 0.0917 1.1 1>
    Crimson=<rgba 0.8 0.0667 0 0.5>
    ForestGreen=<rgba 0.09 0.45 0.1082 1>
    Grey=<rgba 0.25 0.25 0.25 1.25>
    Lime=<rgba 0.3 0.45 0 0.6>
    Orange=<rgba 0.5 0.1917 0 1>
    Pink=<rgba 0.4 0 0.2 0.45>
    Purple=<rgba 0.15 0 0.75 0.7>
    R1 Black:LinearColor=<rgba 0.0608 0.0608 0.0608 3>
    R1 BurntSienna:LinearColor=<rgba 0.2 0.0867 0.03 3.5>
    R1 Cobalt:LinearColor=<rgba 0 0.0092 0.5 0.75>
    R1 Crimson:LinearColor=<rgba 0.3 0.025 0 1.2>
    R1 ForestGreen:LinearColor=<rgba 0 0.2 0.0167 1>
    R1 Grey:LinearColor=<rgba 0.14 0.14 0.14 1.3>
    R1 Lime:LinearColor=<rgba 0.1458 0.25 0 1>
    R1 Orange:LinearColor=<rgba 0.5 0.0667 0 1>
    R1 Pink:LinearColor=<rgba 0.5 0.1 0.4333 0.6>
    R1 Purple:LinearColor=<rgba 0.1333 0 0.4 2>
    R1 Saffron:LinearColor=<rgba 0.45 0.3 0 0.8>
    R1 SkyBlue:LinearColor=<rgba 0.16 0.5858 0.8 0.25>
    R1 TitaniumWhite:LinearColor=<rgba 0.35 0.35 0.35 0.35>
    Saffron=<rgba 0.55 0.44 0 0.4>
    SkyBlue=<rgba 0.22 0.44 0.55 0.6>
    TitaniumWhite=<rgba 0.4 0.4 0.4 0.7>
    Use R1 Painted Overrides=true
   )
   &VC_InitializeRibbon_v1(
    Color=$Particles.Color
    Custom Offset=true
    Enable Skeletal Position=true
    Lifetime=&MergeHelper_FloatSwitch(0=0.9, 1=1.1, 2=0.55, 3=?, Index=0, Num=::NewEnumerator1)
    Local Vector Offset=&AddVectorToPosition(Position=<25 0 -15>, Vector=$User.offset)
    Offset Amount=1
    Ribbon Width=4
   )
   &AddVelocity(Velocity=<-1000 0 0>) {module=AddVelocity001}
  ]
  particleUpdateScript [
   &FN_DisableTireFX(Mask Index Attribute:NiagaraInt32=$Particles.SpawnGroup, Mask Index FL=2, Mask Index FR=3, Mask Index RL=0, Mask Index RR=1)
   &InheritSourceMovement(
    Applied Position Scale=&VectorFromFloat(Value=&FloatFromCurve(
      FloatCurve=DataInterface() {propertyValues {Curve {
         defaultValue=1
         keys [<at 0 1 RCIM_Cubic>, {arriveTangent=-3.2426, interpMode=RCIM_Cubic, leaveTangent=-3.2426, tangentMode=RCTM_User, time=1, value=0}]
        }}}
      Scale Curve=1.1
     ))
    ApplyForce=false
    ApplyPosition=true
    ApplyPositionFalloff=false
   )
   &ParticleState()
   &SolveForcesAndVelocity(Clamp Velocity=false, Limit Acceleration=false)
   &VC_UpdateRibbon_v1(Enable Skeletal Position=true)
   &ScaleRibbonWidth(Ribbon Width Scale=&MakeCustomFloatFromBool(
     Bool=$User.bIsPainted
     True Float=&FloatFromCurve(
      FloatCurve=DataInterface() {propertyValues {Curve {keys [
          <at 0 1.1 RCIM_Cubic>
          {arriveTangent=0.9918, interpMode=RCIM_Cubic, leaveTangent=0.9918, tangentMode=RCTM_User, time=0.5057, value=0.5}
          <at 1 0.2 RCIM_Cubic>
         ]}}}
      Scale Curve=20
     )
     False Float=&FloatFromCurve(
      FloatCurve=DataInterface() {propertyValues {Curve {
         defaultValue=0
         keys [
          <at 0 0.5 RCIM_Cubic>
          {arriveTangent=-0.3934, interpMode=RCIM_Cubic, leaveTangent=-0.3934, tangentMode=RCTM_User, time=0.35, value=0.6}
          {arriveTangent=-0.1896, interpMode=RCIM_Cubic, leaveTangent=-0.1896, tangentMode=RCTM_User, time=1, value=0}
         ]
        }}}
      Scale Curve=30
     )
    ))
   /SetVariables(
    Particles.BC_Principal:LinearColor=&MakeCustomLinearColorFromBool(
     Bool=$User.bIsPainted
     True Color=&MultiplyLinearColor(
      Linear Color=$Particles.Color
      Scale Factor=&MakeVector4FromVectorAndFloat(
       "Vector (RGB)"=&VectorFromCurve(VectorCurve=DataInterface() {propertyValues {
          XCurve {keys [
            <at 0 1.25 RCIM_Cubic>
            {arriveTangent=-0.5, interpMode=RCIM_Cubic, leaveTangent=-0.5, time=0.2, value=0.7}
            {arriveTangent=-1.1667, interpMode=RCIM_Cubic, leaveTangent=-1.1667, time=0.5, value=1}
            {interpMode=RCIM_Cubic, leaveTangent=-4, time=0.8, value=0}
           ]}
          YCurve {keys [
            <at 0 1.25 RCIM_Cubic>
            {arriveTangent=-0.5, interpMode=RCIM_Cubic, leaveTangent=-0.5, time=0.2, value=0.7}
            {arriveTangent=-1.1667, interpMode=RCIM_Cubic, leaveTangent=-1.1667, time=0.5, value=1}
            {interpMode=RCIM_Cubic, leaveTangent=-4, time=0.8, value=0}
           ]}
          ZCurve {keys [
            <at 0 1.25 RCIM_Cubic>
            {arriveTangent=-0.5, interpMode=RCIM_Cubic, leaveTangent=-0.5, time=0.2, value=0.7}
            {arriveTangent=-1.1667, interpMode=RCIM_Cubic, leaveTangent=-1.1667, time=0.5, value=1}
            {interpMode=RCIM_Cubic, leaveTangent=-4, time=0.8, value=0}
           ]}
         }})
       "Float (Alpha)"=&FloatFromCurve(
        FloatCurve=DataInterface() {propertyValues {Curve {
           defaultValue=1
           keys [
            <at 0.05 0 RCIM_Cubic>
            {arriveTangent=0.5984, interpMode=RCIM_Cubic, leaveTangent=0.5984, tangentMode=RCTM_User, time=0.25, value=0.5}
            {arriveTangent=-1.6514, interpMode=RCIM_Cubic, leaveTangent=-1.6514, time=0.35, value=0.5}
            {arriveTangent=4.5249, interpMode=RCIM_Cubic, leaveTangent=4.5249, tangentMode=RCTM_User, time=0.5528, value=0}
            <at 0.7 0.5 RCIM_Cubic>
            {arriveTangent=-0.0346, interpMode=RCIM_Cubic, leaveTangent=-0.0346, tangentMode=RCTM_User, time=1, value=0}
           ]
          }}}
        Scale Curve=3
       )
      )
     )
     False Color=&MultiplyLinearColor(
      Linear Color=&ColorFromCurve(ColorCurve=DataInterface() {propertyValues {
         AlphaCurve {keys [<at 0.0087 0>,<at 0.1807 0.6085>,<at 0.4646 0.4843>,<at 1 0 RCIM_Cubic>]}
         BlueCurve {keys [<at 0.0349 2 RCIM_Cubic>,<at 0.4576 0.8 RCIM_Cubic>]}
         GreenCurve {keys [<at 0.0349 1.125 RCIM_Cubic>,<at 0.4576 0 RCIM_Cubic>]}
         RedCurve {keys [<at 0.0349 0.5 RCIM_Cubic>,<at 0.4576 0.12 RCIM_Cubic>]}
        }})
      Scale Factor=&MakeVector4FromVectorAndFloat(
       "Vector (RGB)"=&VectorFromFloat(Value=&Waveform(
         Global Amplitude Scale=0.8
         "Waveform [1]"=::NewEnumerator0
         "[1] Frequency"=0.8
         "[1] Amplitude Min/Max"=<0 0.7>
         "[1] Scale"=0.5
         "[1] Bias"=0.8
        ))
       "Float (Alpha)"=&FloatFromCurve(
        FloatCurve=DataInterface() {propertyValues {Curve {
           defaultValue=1
           keys [
            {interpMode=RCIM_Cubic, tangentMode=RCTM_User, time=0, value=0.5}
            {interpMode=RCIM_Cubic, tangentMode=RCTM_User, time=0.5, value=1}
            {tangentMode=RCTM_User, time=1, value=0}
           ]
          }}}
        Scale Curve=&MakeCustomFloatFromBool(Bool=$User.bIsFrontendPreview, True Float=0.5, False Float=1.5)
       )
      )
     )
    )
    Particles.BC_Second:LinearColor=&MakeCustomLinearColorFromBool(
     Bool=$User.bIsPainted
     True Color=&MultiplyLinearColor(
      Linear Color=$Particles.R1 Color
      Scale Factor=&MakeVector4FromVectorAndFloat(
       "Vector (RGB)"=&VectorFromFloat(Value=&FloatFromCurve(FloatCurve=DataInterface() {propertyValues {Curve {keys [<at 0 1.5 RCIM_Cubic>,<at 0.5 1 RCIM_Cubic>]}}}))
       "Float (Alpha)"=&FloatFromCurve(
        FloatCurve=DataInterface() {propertyValues {Curve {
           defaultValue=0
           keys [
            {interpMode=RCIM_Cubic, tangentMode=RCTM_User, time=0, value=0.25}
            {interpMode=RCIM_Cubic, tangentMode=RCTM_User, time=0.25, value=0.25}
            {arriveTangent=-0.777, interpMode=RCIM_Cubic, leaveTangent=-0.777, tangentMode=RCTM_User, time=0.6, value=0.2}
            {arriveTangent=-0.0516, tangentMode=RCTM_User, time=1, value=0}
           ]
          }}}
        Scale Curve=1.5
       )
      )
     )
     False Color=&MultiplyLinearColor(
      Linear Color=&ColorFromCurve(ColorCurve=DataInterface() {propertyValues {
         AlphaCurve {keys [<at 0.0087 0>,<at 0.1807 0.6085>,<at 0.4648 0.4>,<at 1 0 RCIM_Cubic>]}
         BlueCurve {keys [<at 0 2 RCIM_Cubic>,<at 0.4576 0.8 RCIM_Cubic>]}
         GreenCurve {keys [<at 0 1.125 RCIM_Cubic>,<at 0.4576 0 RCIM_Cubic>]}
         RedCurve {keys [<at 0 0.5 RCIM_Cubic>,<at 0.4576 0.12 RCIM_Cubic>]}
        }})
      Scale Factor=&MakeVector4FromVectorAndFloat(
       "Vector (RGB)"=&VectorFromFloat(Value=&Waveform(
         Global Amplitude Scale=0.8
         "Waveform [1]"=::NewEnumerator0
         "[1] Frequency"=0.8
         "[1] Amplitude Min/Max"=<0 0.7>
         "[1] Scale"=0.5
         "[1] Bias"=0.8
        ))
       "Float (Alpha)"=&FloatFromCurve(
        FloatCurve=DataInterface() {propertyValues {Curve {
           defaultValue=1
           keys [
            {interpMode=RCIM_Cubic, tangentMode=RCTM_User, time=0, value=0.5}
            {interpMode=RCIM_Cubic, tangentMode=RCTM_User, time=0.5, value=1}
            {tangentMode=RCTM_User, time=1, value=0}
           ]
          }}}
        Scale Curve=0.5
       )
      )
     )
    )
    "Particles.RibbonWidth_(3)":NiagaraFloat=&MakeCustomFloatFromBool(
     Bool=$User.bIsPainted
     True Float=&FloatFromCurve(
      FloatCurve=DataInterface() {propertyValues {Curve {keys [<at 0.5 1 RCIM_Cubic>, {arriveTangent=-2, interpMode=RCIM_Cubic, leaveTangent=-2, time=0.8, value=0.5695}, <at 1 0 RCIM_Cubic>]}}}
      Scale Curve=46
     )
     False Float=&FloatFromCurve(
      FloatCurve=DataInterface() {propertyValues {Curve {keys [
          {arriveTangent=-0.5501, interpMode=RCIM_Cubic, time=0, value=0.5}
          {arriveTangent=-0.5, interpMode=RCIM_Cubic, leaveTangent=-0.5, time=0.3, value=0.9611}
          {arriveTangent=0.0376, interpMode=RCIM_Cubic, leaveTangent=0.0376, tangentMode=RCTM_User, time=1, value=0}
         ]}}}
      Scale Curve=40
     )
    )
   ) {isSetParameters=true, module=SetVariables_#2}
  ]
  rendererClasses [&NiagaraRibbonRendererProperties]
  renderers [
   {rendererClass=&NiagaraRibbonRendererProperties}
   {rendererClass=&NiagaraRibbonRendererProperties, rendererIndex=1}
  ]
 }
 NE_VC_DriftTrail_Ribbon_Parent {
  bEnabled=false
  emitterUpdateScript [
   &EmitterState()
   &VC_RateQualityHandler(
    Epic Spawn Rate:NiagaraFloat=&MergeHelper_FloatSwitch(0=45, 1=15, 2=45, 3=?, Index=0, Num=::NewEnumerator1)
    High Spawn Rate:NiagaraFloat=&MergeHelper_FloatSwitch(0=35, 1=15, 2=40, 3=?, Index=0, Num=::NewEnumerator1)
    Low Spawn Rate:NiagaraFloat=&MergeHelper_FloatSwitch(0=32, 1=15, 2=30, 3=?, Index=0, Num=::NewEnumerator1)
    Medium Spawn Rate:NiagaraFloat=&MergeHelper_FloatSwitch(0=31, 1=15, 2=35, 3=?, Index=0, Num=::NewEnumerator1)
    Use Custom Rate?=true
   )
   &VC_SpawnRibbon_v1(Number of Ribbons=2, SpawnRate=$Emitter.Spawn Rate)
  ]
  particleSpawnScript [
   &VC_PaintedVariantsOverride(
    Black=<rgba 0.1 0.1 0.1 1.5>
    BurntSienna=<rgba 0.4948 0.1555 0.0644 1>
    Cobalt=<rgba 0.3293 0.3329 0.5226 1>
    Crimson=<rgba 1.4132 0.0957 0.1314 1>
    ForestGreen=<rgba 0.1441 0.5851 0.1664 1>
    Grey=<rgba 0.2587 0.2587 0.2587 2>
    Lime=<rgba 0.2196 0.4392 0 3>
    Orange=<rgba 0.5885 0.2053 0 3>
    Pink=<rgba 2.444 0.0467 1.5722 1>
    Purple=<rgba 0.2832 0.1142 0.526 1>
    Saffron=<rgba 0.8715 0.6682 0 1>
    SkyBlue=<rgba 0.3345 0.8362 1.0885 1.5>
    TitaniumWhite=<rgba 0.5503 0.5503 0.5503 1>
   )
   &VC_InitializeRibbon_v1(
    Color=$Particles.Color
    Custom Offset=true
    Enable Skeletal Position=true
    Lifetime=&MergeHelper_FloatSwitch(0=0.9, 1=1.1, 2=0.55, 3=?, Index=0, Num=::NewEnumerator1)
    Local Vector Offset=<-70 0 -25>
    Offset Amount=1
    Ribbon Width=1.5
   )
   &AddVelocity(Velocity=<-1000 0 0>) {module=AddVelocity001}
  ]
  particleUpdateScript [
   &FN_DisableTireFX(Mask Index Attribute:NiagaraInt32=$Particles.SpawnGroup, Mask Index FL=2, Mask Index FR=3, Mask Index RL=0, Mask Index RR=1)
   &InheritSourceMovement(
    Applied Position Scale=&VectorFromFloat(Value=&FloatFromCurve(
      FloatCurve=DataInterface() {propertyValues {Curve {
         defaultValue=1
         keys [<at 0 1 RCIM_Cubic>, {arriveTangent=-3.1129, interpMode=RCIM_Cubic, leaveTangent=-3.1129, tangentMode=RCTM_User, time=1, value=0}]
        }}}
      Scale Curve=1.1
     ))
    ApplyForce=false
    ApplyPosition=true
    ApplyPositionFalloff=false
   )
   &ParticleState()
   &SolveForcesAndVelocity(Clamp Velocity=false, Limit Acceleration=false)
   &Color(
    Color=&MakeCustomLinearColorFromBool(
     Bool=$User.bIsPainted
     True Color=$Particles.Color
     False Color=&ColorFromCurve(ColorCurve=DataInterface() {propertyValues {
        AlphaCurve {keys [<at 0 1 RCIM_Cubic>,<at 1 1 RCIM_Cubic>]}
        BlueCurve {keys [<at 0 1.5 RCIM_Cubic>,<at 0.4972 1>]}
        GreenCurve {keys [<at 0 1.1622 RCIM_Cubic>,<at 0.4972 0.1892>]}
        RedCurve {keys [<at 0 0.1745 RCIM_Cubic>,<at 0.4972 0.3416>]}
       }})
    )
    Scale Alpha=&MergeHelper_FloatSwitch(0=1, 1=1, 2=3, 3=?, Index=0, Num=::NewEnumerator1)
   )
   &VC_UpdateRibbon_v1(Enable Skeletal Position=true)
  ]
  rendererClasses [&NiagaraRibbonRendererProperties]
  renderers [
   {rendererClass=&NiagaraRibbonRendererProperties}
   {rendererClass=&NiagaraRibbonRendererProperties, rendererIndex=1}
   {rendererClass=&NiagaraRibbonRendererProperties, rendererIndex=2}
   {rendererClass=&NiagaraRibbonRendererProperties, rendererIndex=3}
  ]
 }
 NE_VC_DriftTrail_Ribbon_Parent001 {
  emitterUpdateScript [
   &EmitterState()
   &VC_RateQualityHandler(Use Custom Rate?=true)
   &VC_SpawnRibbon_v1(Number of Ribbons=2, SpawnRate=$Emitter.Spawn Rate)
  ]
  particleSpawnScript [
   &VC_PaintedVariantsOverride(
    Black=<rgba 0.09 0.09 0.09 1.5>
    BurntSienna=<rgba 0.5 0.1458 0.075 1>
    Cobalt=<rgba 0.1743 0.3098 0.5809 1>
    Crimson=<rgba 1 0.15 0.1783 0.65>
    ForestGreen=<rgba 0.1638 0.5851 0.1851 0.5>
    Grey=<rgba 0.22 0.22 0.22 1>
    Lime=<rgba 0.3015 0.45 0.045 0.7>
    Orange=<rgba 2 0.2033 0.04 1>
    Pink=<rgba 1 0.17 0.7233 1>
    Purple=<rgba 0.24 0.11 0.5 1>
    Saffron=<rgba 0.8715 0.5955 0.0436 0.5>
    SkyBlue=<rgba 0.2 0.5992 0.8 1>
    TitaniumWhite=<rgba 0.5503 0.5503 0.5503 0.5>
   )
   &VC_InitializeRibbon_v1(
    Color=$Particles.Color
    Custom Offset=true
    Enable Skeletal Position=true
    Lifetime=&MergeHelper_FloatSwitch(0=0.9, 1=1.1, 2=0.55, 3=?, Index=0, Num=::NewEnumerator1)
    Local Vector Offset=&AddVectorToPosition(Position=<10 0 -15>, Vector=$User.offset)
    Offset Amount=1
    Ribbon Width=1
   )
   &AddVelocity(Velocity=<-1000 0 0>) {module=AddVelocity001}
  ]
  particleUpdateScript [
   &FN_DisableTireFX(Mask Index Attribute:NiagaraInt32=$Particles.SpawnGroup, Mask Index FL=2, Mask Index FR=3, Mask Index RL=0, Mask Index RR=1)
   &InheritSourceMovement(
    Applied Position Scale=&VectorFromFloat(Value=&FloatFromCurve(
      FloatCurve=DataInterface() {propertyValues {Curve {
         defaultValue=1
         keys [<at 0 1 RCIM_Cubic>, {arriveTangent=-3.1129, interpMode=RCIM_Cubic, leaveTangent=-3.1129, tangentMode=RCTM_User, time=1, value=0}]
        }}}
      Scale Curve=1.1
     ))
    ApplyForce=false
    ApplyPosition=true
    ApplyPositionFalloff=false
   )
   &ParticleState()
   &SolveForcesAndVelocity(Clamp Velocity=false, Limit Acceleration=false)
   &Color(
    Color=&MakeCustomLinearColorFromBool(
     Bool=$User.bIsPainted
     True Color=&MultiplyLinearColor(
      Linear Color=$Particles.Color
      Scale Factor=&MakeVector4FromVectorAndFloat(
       "Vector (RGB)"=&VectorFromCurve(VectorCurve=DataInterface() {propertyValues {
          XCurve {keys [<at 0 1 RCIM_Cubic>,<at 1 0.75 RCIM_Cubic>]}
          YCurve {keys [<at 0 1 RCIM_Cubic>,<at 1 0.75 RCIM_Cubic>]}
          ZCurve {keys [<at 0 1 RCIM_Cubic>,<at 1 0.75 RCIM_Cubic>]}
         }})
       "Float (Alpha)"=1
      )
     )
     False Color=&MultiplyLinearColor(
      Linear Color=&ColorFromCurve(ColorCurve=DataInterface() {propertyValues {
         AlphaCurve {keys [<at 0 0.5 RCIM_Cubic>,<at 1 0 RCIM_Cubic>]}
         BlueCurve {keys [<at 0 2 RCIM_Cubic>,<at 0.2 1>,<at 0.419 2>,<at 0.9743 2>]}
         GreenCurve {keys [<at 0 1.6432 RCIM_Cubic>,<at 0.2 0.5799>,<at 0.419 0>,<at 0.9743 0>]}
         RedCurve {keys [<at 0 0.6 RCIM_Cubic>,<at 0.2 0.1457>,<at 0.419 0.1667>,<at 0.9743 0>]}
        }})
      Scale Factor=&MakeVector4FromVectorAndFloat(
       "Vector (RGB)"=&VectorFromFloat(Value=&Waveform(Global Amplitude Scale=1.8, "Waveform [1]"=::NewEnumerator0, "[1] Frequency"=0.8, "[1] Amplitude Min/Max"=<0 2>, "[1] Bias"=1))
       "Float (Alpha)"=&FloatFromCurve(
        FloatCurve=DataInterface() {propertyValues {Curve {
           defaultValue=1
           keys [
            {interpMode=RCIM_Cubic, tangentMode=RCTM_User, time=0, value=0.5}
            {interpMode=RCIM_Cubic, tangentMode=RCTM_User, time=0.5, value=1}
            {tangentMode=RCTM_User, time=1, value=0}
           ]
          }}}
        Scale Curve=0.5
       )
      )
     )
    )
    Scale Alpha=&MergeHelper_FloatSwitch(0=1, 1=1, 2=3, 3=?, Index=0, Num=::NewEnumerator1)
   )
   &VC_UpdateRibbon_v1(Enable Skeletal Position=true)
  ]
  rendererClasses [&NiagaraRibbonRendererProperties]
  renderers [
   {rendererClass=&NiagaraRibbonRendererProperties}
   {rendererClass=&NiagaraRibbonRendererProperties, rendererIndex=1}
   {rendererClass=&NiagaraRibbonRendererProperties, rendererIndex=2}
   {rendererClass=&NiagaraRibbonRendererProperties, rendererIndex=3}
  ]
 }
 NE_VC_DriftTrail_Sprite_Parent {
  bEnabled=false
  emitterUpdateScript [
   &EmitterState(
    Inactive Response=::NewEnumerator1
    Life Cycle Mode=::NewEnumerator1
    Loop Behavior=::NewEnumerator1
    Loop Duration Mode=::NewEnumerator1
    UseLoopDelay=false
   )
   &SpawnBurst_Instantaneous(Spawn Count=8, Use Spawn Probability=false, UseLoopCountLimit=false)
   &SpawnRate(SpawnRate=50, Use Spawn Probability=false) {enabled=false}
  ]
  particleSpawnScript [
   &VC_PaintedVariantsOverride() {enabled=false}
   &VC_PaintedVariantsOverride(
    Black=<rgba 0.092 0.092 0.092 1>
    BurntSienna=<rgba 0.7205 0.2225 0.0888 1>
    Cobalt=<rgba 0.3 0.4054 6 1>
    Crimson=<rgba 3 0.087 0.166 1>
    DefaultColor=<rgba 0.5 1.25 5 1>
    ForestGreen=<rgba 0 1.6 0.0808 1>
    Gold:LinearColor=<rgba 2 2 2 1>
    Grey=<rgba 0.3281 0.3281 0.3281 1>
    Lime=<rgba 0.625 1.25 0 1>
    Orange=<rgba 1.9566 0.1725 0 1>
    Pink=<rgba 1.6 0.1917 1.0879 1>
    Purple=<rgba 0.2632 0 3 1>
    Saffron=<rgba 1.5 1.15 0 1>
    SkyBlue=<rgba 0.9453 2.1011 2.6823 1>
    TitaniumWhite=<rgba 1.1557 1.1557 1.1557 1>
   ) {module=VC_PaintedVariantsOverride001}
   &InitializeParticle(Color=$Particles.Color, Sprite Size Mode=::NewEnumerator3, Uniform Sprite Size=20, UsePositionOffset=false, Write Material Random=true)
   &PartitionParticles()
   &SkeletalMeshLocation(
    Bone / Socket Index=&Add_Integer(A=$Particles.PartitionParticles.Partition, B=4)
    Bone / Socket Sampling Mode=::NewEnumerator5
    Interpolate Mesh Sampling=false
    MaskSpawnGroup=false
    Mesh Sampling Type=::NewEnumerator6
    Sampled Position Offset=&AddVectorToPosition(Position=<0 0 0>, Vector=$User.offset)
    Skeletal Mesh=$System.SkeletalMesh
   )
   &AddVelocity(Velocity=<-1500 0 0>) {enabled=false}
  ]
  particleUpdateScript [
   &ShapeLocation(
    MaskSpawnGroup=false
    Offset=&MakeVector(
     X=&Add_Float(
      A=&Add_Float(
       A=&MakeFloatFromVector(Vector=$User.offset, Channel=::NewEnumerator0)
       B=&MakeFloatFromVector(Vector=<0 0 0>, Channel=::NewEnumerator0)
      )
      B=&Sine(Normalized Angle=$Emitter.Age, Period=0.5, Scale=0.3, Bias=0)
     )
     Y=&Add_Float(
      A=&Add_Float(
       A=&MakeFloatFromVector(Vector=$User.offset, Channel=::NewEnumerator1)
       B=&MakeFloatFromVector(Vector=<0 0 0>, Channel=::NewEnumerator1)
      )
      B=&Sine(Normalized Angle=$Emitter.Age, Period=1, Scale=0.3, Bias=0)
     )
     Z=&Add_Float(
      A=&Add_Float(
       A=&MakeFloatFromVector(Vector=$User.offset, Channel=::NewEnumerator2)
       B=&MakeFloatFromVector(Vector=<0 0 0>, Channel=::NewEnumerator2)
      )
      B=&Sine(Normalized Angle=$Emitter.Age, Period=0.75, Scale=0.3, Bias=0)
     )
    )
    Offset Mode=::NewEnumerator0
    Override Seed=false
    Ring Radius=0
    Shape Origin=$Particles.Position
    Shape Primitive=::NewEnumerator4
    UseOwnerScale=true
   ) {enabled=false}
   &ParticleState(Kill Particles When Lifetime Has Elapsed=false, Loop Particles Lifetime=true)
   &ScaleSpriteSize(Scale Sprite Size Mode=::NewEnumerator0, Uniform Scale Factor=3.5)
   &SolveForcesAndVelocity(Clamp Velocity=false, Limit Acceleration=false)
   &SubUVAnimation(
    "Sprite Renderer "=DataInterface() {propertyValues {SpriteRenderer=&self:NE_VC_DriftTrail_Sprite_Parent_0.NiagaraSpriteRendererProperties_3}}
    SubUV Lookup Index Scale=3
    UseEndFrame=false
    UseStartFrame=false
   )
  ]
  rendererClasses [&NiagaraSpriteRendererProperties]
  renderers [
   {rendererClass=&NiagaraSpriteRendererProperties}
   {rendererClass=&NiagaraSpriteRendererProperties, rendererIndex=1}
  ]
 }
 NE_VC_DriftTrail_Sprite_Parent001 {
  emitterUpdateScript [
   &EmitterState(
    Inactive Response=::NewEnumerator1
    Life Cycle Mode=::NewEnumerator1
    Loop Behavior=::NewEnumerator1
    Loop Duration Mode=::NewEnumerator1
    UseLoopDelay=false
   )
   &SpawnBurst_Instantaneous(Spawn Count=8, Use Spawn Probability=false, UseLoopCountLimit=false)
   &SpawnRate(SpawnRate=50, Use Spawn Probability=false) {enabled=false}
  ]
  particleSpawnScript [
   &VC_PaintedVariantsOverride() {enabled=false}
   &VC_PaintedVariantsOverride(
    Black=<rgba 0.06 0.06 0.06 1>
    BurntSienna=<rgba 1 0.1925 0.05 1>
    Cobalt=<rgba 0.18 0.2877 6 1>
    Crimson=<rgba 2.5 0.0875 0.1277 1>
    DefaultColor=<rgba 0.5 1.25 5 1>
    ForestGreen=<rgba 0.16 1.6 0.28 1>
    Gold:LinearColor=<rgba 2 2 2 1>
    Grey=<rgba 0.4 0.4 0.4 2>
    Lime=<rgba 1 2 0 1>
    Orange=<rgba 3 0.2 0 1>
    Pink=<rgba 2 0.3 2 1>
    Purple=<rgba 0.225 0 2.25 1>
    Saffron=<rgba 1.5 1.175 0 1>
    SkyBlue=<rgba 0.4023 1.5423 2.6823 1>
    TitaniumWhite=<rgba 2 2 2 1>
   ) {module=VC_PaintedVariantsOverride001}
   &InitializeParticle(Color=$Particles.Color, Sprite Size Mode=::NewEnumerator3, Uniform Sprite Size=20, UsePositionOffset=false, Write Material Random=true)
   &PartitionParticles()
   &SkeletalMeshLocation(
    Bone / Socket Index=&Add_Integer(A=$Particles.PartitionParticles.Partition, B=4)
    Bone / Socket Sampling Mode=::NewEnumerator5
    Interpolate Mesh Sampling=false
    MaskSpawnGroup=false
    Mesh Sampling Type=::NewEnumerator6
    Sampled Position Offset=&AddVectorToPosition(Position=<0 0 0>, Vector=$User.offset)
    Skeletal Mesh=$System.SkeletalMesh
   )
   &AddVelocity(Velocity=<-1500 0 0>) {enabled=false}
  ]
  particleUpdateScript [
   &ParticleState(Kill Particles When Lifetime Has Elapsed=false, Loop Particles Lifetime=true)
   &ScaleSpriteSize(Scale Sprite Size Mode=::NewEnumerator0, Uniform Scale Factor=3.5)
   &SolveForcesAndVelocity(Clamp Velocity=false, Limit Acceleration=false)
   &SubUVAnimation(
    "Sprite Renderer "=DataInterface() {propertyValues {SpriteRenderer=&self:NE_VC_DriftTrail_Sprite_Parent001_0.NiagaraSpriteRendererProperties_3}}
    SubUV Lookup Index Scale=3
    UseEndFrame=false
    UseStartFrame=false
   )
  ]
  rendererClasses [&NiagaraSpriteRendererProperties]
  renderers [
   {rendererClass=&NiagaraSpriteRendererProperties}
   {rendererClass=&NiagaraSpriteRendererProperties, rendererIndex=1}
  ]
 }
 NE_VC_DriftTrail_Sprite_Parent002 {
  bEnabled=false
  emitterUpdateScript [
   &EmitterState()
   &SpawnRate(SpawnRate=35, Use Spawn Probability=false)
  ]
  particleSpawnScript [
   &VC_PaintedVariantsOverride(
    Black=<rgba 0.0816 0.0816 0.0816 1>
    BurntSienna=<rgba 0.1615 0.0954 0.0776 1>
    Cobalt=<rgba 0.1148 0.1236 0.5955 1>
    Crimson=<rgba 0.5747 0.0688 0.0825 1>
    ForestGreen=<rgba 0 0.467 0.0236 1>
    Grey=<rgba 0.1649 0.1649 0.1649 2>
    Lime=<rgba 0.1333 0.1997 0.0669 1>
    Orange=<rgba 0.5434 0.0707 0 2>
    Pink=<rgba 0.4149 0.0771 0.2921 1>
    Purple=<rgba 0.1114 0.0995 0.2691 1>
    Saffron=<rgba 0.4184 0.3208 0 1>
    SkyBlue=<rgba 0.1886 0.3993 0.5052 1>
    TitaniumWhite=<rgba 0.3188 0.3188 0.3188 1>
   )
   &InitializeParticle(
    Color=$Particles.Color
    Lifetime=&MakeCustomFloatFromBool(Bool=$User.bIsFrontendPreview, True Float=&RandomRangeFloat(Minimum=0.5, Maximum=1), False Float=1)
    Sprite Size Mode=::NewEnumerator3
    Uniform Sprite Size=&RandomRangeFloat(Minimum=8, Maximum=12)
    UsePositionOffset=false
    Write Material Random=true
   )
   &PartitionParticles()
   &SkeletalMeshLocation(
    Bone / Socket Index=&Add_Integer(A=$Particles.PartitionParticles.Partition, B=4)
    Bone / Socket Sampling Mode=::NewEnumerator5
    Interpolate Mesh Sampling=false
    MaskSpawnGroup=false
    Mesh Sampling Type=::NewEnumerator6
    Sampled Position Offset=&AddVectorToPosition(Position=<0 0 -15>, Vector=$User.offset)
    Skeletal Mesh=$System.SkeletalMesh
   )
   &AddVelocity(Velocity=&MakeCustomVectorFromBool(
     Bool=$User.bIsFrontendPreview
     True Vector=&RandomRangeVector(Minimum=<-100 0 20>, Maximum=<-70 0 50>)
     False Vector=&RandomRangeVector(Minimum=<-80 0 0>, Maximum=<0 0 0>)
    ))
   &ShapeLocation(MaskSpawnGroup=false, Override Seed=false, Sphere Radius=13, UseOwnerScale=true) {module=ShapeLocation001}
  ]
  particleUpdateScript [
   &SubUVAnimation(
    "Sprite Renderer "=DataInterface() {propertyValues {SpriteRenderer=&self:NE_VC_DriftTrail_Sprite_Parent002_0.NiagaraSpriteRendererProperties_1}}
    UseEndFrame=false
    UseStartFrame=false
   )
   &ParticleState()
   &ScaleSpriteSize(
    Uniform Curve Scale=3.2
    Uniform Curve Sprite Scale=DataInterface() {propertyValues {Curve {keys [{tangentMode=RCTM_User, time=0, value=0}, {tangentMode=RCTM_User, time=0.7, value=1}]}}}
   )
   &SolveForcesAndVelocity(Clamp Velocity=false, Limit Acceleration=false)
   &Color(
    Color=&MakeCustomLinearColorFromBool(Bool=$User.bIsPainted, True Color=$Particles.Color, False Color=<rgba 0 0.15 0.9 0.5>)
    Scale Alpha=&FloatFromCurve(FloatCurve=DataInterface() {propertyValues {Curve {keys [
         {arriveTangent=-0.504, interpMode=RCIM_Cubic, leaveTangent=-0.504, tangentMode=RCTM_User, time=0.3, value=1}
         {arriveTangent=-2.2923, interpMode=RCIM_Cubic, leaveTangent=-2.2923, tangentMode=RCTM_User, time=0.8191, value=0.3709}
         {arriveTangent=-0.7856, interpMode=RCIM_Cubic, leaveTangent=-0.7856, tangentMode=RCTM_User, time=1, value=0}
        ]}}})
   )
  ]
  rendererClasses [&NiagaraSpriteRendererProperties]
  renderers [
   {rendererClass=&NiagaraSpriteRendererProperties}
   {rendererClass=&NiagaraSpriteRendererProperties, rendererIndex=1}
  ]
 }
 NE_VC_DriftTrail_Sprite_Parent003 {
  emitterUpdateScript [
   &EmitterState(Inactive Response=::NewEnumerator1, Life Cycle Mode=::NewEnumerator1, Loop Duration=1, UseLoopDelay=false)
   &SpawnRate(SpawnRate=25, Use Spawn Probability=false)
  ]
  particleSpawnScript [
   &VC_PaintedVariantsOverride(
    Black=<rgba 0.0816 0.0816 0.0816 0.8>
    BurntSienna=<rgba 0.31 0.1421 0.1085 1>
    Cobalt=<rgba 0.1448 0.1952 0.75 1>
    Crimson=<rgba 0.47 0.094 0.1253 1>
    ForestGreen=<rgba 0.12 0.4 0.1433 1>
    Grey=<rgba 0.22 0.22 0.22 1.5>
    Lime=<rgba 0.3295 0.5 0.035 1>
    Orange=<rgba 0.95 0.1236 0 1.5>
    Pink=<rgba 0.5 0.19 0.4742 1>
    Purple=<rgba 0.1416 0.0055 0.55 1>
    Saffron=<rgba 0.55 0.4583 0 1>
    SkyBlue=<rgba 0.2 0.3996 0.5 1>
    TitaniumWhite=<rgba 0.42 0.42 0.42 0.8>
   )
   &InitializeParticle(
    Color=$Particles.Color
    Lifetime=&MakeCustomFloatFromBool(Bool=$User.bIsFrontendPreview, True Float=&RandomRangeFloat(Minimum=0.75, Maximum=0.8), False Float=1)
    Sprite Size Mode=::NewEnumerator3
    Uniform Sprite Size=&RandomRangeFloat(Minimum=9, Maximum=11)
    UsePositionOffset=false
    Write Material Random=true
   )
   &PartitionParticles()
   &SkeletalMeshLocation(
    Bone / Socket Index=&Add_Integer(A=$Particles.PartitionParticles.Partition, B=4)
    Bone / Socket Sampling Mode=::NewEnumerator5
    Interpolate Mesh Sampling=false
    MaskSpawnGroup=false
    Mesh Sampling Type=::NewEnumerator6
    Sampled Position Offset=&AddVectorToPosition(Position=<-10 0 -5>, Vector=$User.offset)
    Skeletal Mesh=$System.SkeletalMesh
   )
   &AddVelocity(Velocity=&MakeCustomVectorFromBool(
     Bool=$User.bIsFrontendPreview
     True Vector=&RandomRangeVector(Minimum=<-80 0 30>, Maximum=<-50 0 50>)
     False Vector=&RandomRangeVector(Minimum=<-80 0 0>, Maximum=<0 0 0>)
    ))
   &ShapeLocation(MaskSpawnGroup=false, Non Uniform Scale=<1.5 1 0.5>, Override Seed=false, Sphere Radius=15, UseOwnerScale=true) {module=ShapeLocation001}
  ]
  particleUpdateScript [
   &SubUVAnimation(
    "Sprite Renderer "=DataInterface() {propertyValues {SpriteRenderer=&self:NE_VC_DriftTrail_Sprite_Parent003_0.NiagaraSpriteRendererProperties_1}}
    UseEndFrame=false
    UseStartFrame=false
   )
   &ParticleState()
   &ScaleSpriteSize(
    Uniform Curve Scale=3.2
    Uniform Curve Sprite Scale=DataInterface() {propertyValues {Curve {keys [{tangentMode=RCTM_User, time=0, value=0.75}, {tangentMode=RCTM_User, time=0.7, value=1.25}]}}}
   )
   &CurlNoiseForce(Bias Noise=false, Noise Strength=1200, Pan Noise=false)
   &ScaleVelocity(Velocity Scale=&VectorFromCurve(
     VectorCurve=DataInterface() {propertyValues {
       XCurve {keys [<at 0 0.4 RCIM_Cubic>,<at 1 1 RCIM_Cubic>]}
       YCurve {keys [<at 0 1>,<at 1 1>]}
       ZCurve {keys [{arriveTangent=0.0352, interpMode=RCIM_Cubic, leaveTangent=0.0352, tangentMode=RCTM_User, time=0.05, value=0}, <at 0.95 1.25 RCIM_Cubic>]}
      }}
     Scale Curve=&SelectVectorFromArray(
      Array Sampling Mode=::NewEnumerator0
      Direct Array Index=?
      Vector Selection Array=DataInterface() {propertyValues {FloatData [<1.15 1 1.2>, <1 1 1>]}}
      Random Seed=0
     )
    ))
   &SolveForcesAndVelocity(Clamp Velocity=false, Limit Acceleration=false)
   &Color(
    Color=&MakeCustomLinearColorFromBool(
     Bool=$User.bIsPainted
     True Color=$Particles.Color
     False Color=&SelectLinearColorFromArray(
      Array Sampling Mode=::NewEnumerator0
      Direct Array Index=?
      Color Selection Array=DataInterface() {propertyValues {ColorData [<rgba 0.125 0.3125 1.25 0.5>, <rgba 0.1 0.25 1 1>]}}
      Random Seed=0
     )
    )
    Scale Alpha=&MakeCustomFloatFromBool(
     Bool=$User.bIsPainted
     True Float=&FloatFromCurve(FloatCurve=DataInterface() {propertyValues {Curve {keys [<at 0.25 1 RCIM_Cubic>, {arriveTangent=-4.4457, interpMode=RCIM_Cubic, leaveTangent=-4.4457, tangentMode=RCTM_User, time=1, value=0}]}}})
     False Float=&FloatFromCurve(
      FloatCurve=DataInterface() {propertyValues {Curve {keys [
          <at 0.0096 1.25 RCIM_Cubic>
          {arriveTangent=-1.2622, interpMode=RCIM_Cubic, leaveTangent=-1.2622, time=0.25, value=1}
          {interpMode=RCIM_Cubic, leaveTangent=-0.7856, time=1, value=0}
         ]}}}
      Scale Curve=&RandomRangeFloat(Minimum=0.5, Maximum=1)
     )
    )
   )
  ]
  rendererClasses [&NiagaraSpriteRendererProperties]
  renderers [
   {rendererClass=&NiagaraSpriteRendererProperties}
   {rendererClass=&NiagaraSpriteRendererProperties, rendererIndex=1}
  ]
 }
 NE_VC_DriftTrail_WheelWrap_Parent {
  bEnabled=false
  emitterUpdateScript [
   &EmitterState(
    Inactive Response=::NewEnumerator1
    Life Cycle Mode=::NewEnumerator1
    Loop Behavior=::NewEnumerator1
    Loop Duration Mode=::NewEnumerator1
    UseLoopDelay=false
   )
   &SpawnBurst_Instantaneous(Spawn Count=4, Use Spawn Probability=false, UseLoopCountLimit=false)
  ]
  particleSpawnScript [
   &VC_PaintedVariantsOverride(
    Black=<rgba 0.2205 0.2205 0.2205 1>
    BurntSienna=<rgba 0.9705 0.4412 0.2725 1>
    Cobalt=<rgba 0.5273 0.5407 1.25 1>
    Crimson=<rgba 1.5 0.3359 0.3675 1>
    DefaultColor=<rgba 0.7043 0.8338 1 0.5>
    ForestGreen=<rgba 0.5143 0.7899 0.5282 1>
    Grey=<rgba 0.7969 0.7969 0.7969 1>
    Lime=<rgba 0.4016 0.724 0.0792 1>
    Orange=<rgba 0.89 0.3104 0 1>
    Pink=<rgba 0.8837 0.4219 0.7158 1>
    R1 Black:LinearColor=<rgba 0.05 0.05 0.05 20>
    R1 BurntSienna:LinearColor=<rgba 0.4566 0.2223 0.1264 1>
    R1 Cobalt:LinearColor=<rgba 0 0.0185 1 1>
    R1 Crimson:LinearColor=<rgba 0.4635 0 0.0126 1>
    R1 DefaultColor=<rgba 0.24 0.66 0.8 1>
    R1 ForestGreen:LinearColor=<rgba 0 0.5816 0.0294 1>
    R1 Grey:LinearColor=<rgba 0.1788 0.1788 0.1788 10>
    R1 Lime:LinearColor=<rgba 0.4362 0.526 0.1799 1>
    R1 Orange:LinearColor=<rgba 0.6724 0.2703 0.0549 1>
    R1 Pink:LinearColor=<rgba 0.3698 0.0443 0.2514 1>
    R1 Purple:LinearColor=<rgba 0.3992 0.2041 0.6795 1>
    R1 Saffron:LinearColor=<rgba 0.5017 0.4345 0.2134 1>
    R1 SkyBlue:LinearColor=<rgba 0.3712 0.6822 0.8385 1>
    R1 TitaniumWhite:LinearColor=<rgba 0.6441 0.6441 0.6441 1>
    R2 Black:LinearColor=<rgba 0.02 0.02 0.02 10>
    R2 BurntSienna:LinearColor=<rgba 0.1302 0.0592 0.0319 1>
    R2 Cobalt:LinearColor=<rgba 0 0.0129 0.6997 1>
    R2 DefaultColor:LinearColor=<rgba 0.1194 0.1991 0.5972 1>
    R2 ForestGreen:LinearColor=<rgba 0 0.2274 0.0115 1>
    R2 Lime:LinearColor=<rgba 0.2714 0.349 0.064 1>
    R2 Orange:LinearColor=<rgba 0.2448 0.0854 0 1>
    R2 Pink:LinearColor=<rgba 0.276 0.1203 0.2194 1>
    R2 Purple:LinearColor=<rgba 0.109 0 0.2656 1>
    R2 Saffron:LinearColor=<rgba 0.1788 0.1371 0 1>
    R2 SkyBlue:LinearColor=<rgba 0 0.2206 0.3316 1>
    R2 TitaniumWhite:LinearColor=<rgba 0.3212 0.3212 0.3212 1>
    Saffron=<rgba 0.6684 0.5124 0 1>
    SkyBlue=<rgba 0.3449 0.7808 1 1>
    TitaniumWhite=<rgba 0.5538 0.5538 0.5538 1>
    Use R1 Painted Overrides=true
    Use R2 Painted Overrides=true
   )
   &InitializeParticle(
    Color=$Particles.Color
    Mesh Scale Mode=::NewEnumerator4
    Sprite Rotation Mode=::NewEnumerator2
    Sprite Size Mode=::NewEnumerator3
    UsePositionOffset=false
    Write Material Random=true
   )
   &PartitionParticles()
   &SkeletalMeshLocation(
    Bone / Socket Index=&Add_Integer(A=$Particles.PartitionParticles.Partition, B=4)
    Bone / Socket Sampling Mode=::NewEnumerator5
    Interpolate Mesh Sampling=false
    MaskSpawnGroup=false
    Mesh Sampling Type=::NewEnumerator6
    Skeletal Mesh=$System.SkeletalMesh
   )
  ]
  particleUpdateScript [
   &FN_DisableTireFX(
    Mask Index Attribute:NiagaraInt32=$Particles.PartitionParticles.Partition
    Mask Index FL=2
    Mask Index FR=3
    Mask Index RL=0
    Mask Index RR=1
   )
   &VC_WheelWrapScale(StandardRadius=49, Wheel Index Attribute:NiagaraInt32=&Add_Integer(A=$Particles.PartitionParticles.Partition, B=2))
   &ParticleState(Kill Particles When Lifetime Has Elapsed=false)
   &UpdateMeshOrientation(Rotation Coordinate Space=::NewEnumerator2, Rotation Vector=<0 1 0>) {enabled=false}
   &SkeletalMeshLocation(
    Bone / Socket Index=&Add_Integer(A=$Particles.PartitionParticles.Partition, B=4)
    Bone / Socket Sampling Mode=::NewEnumerator5
    MaskSpawnGroup=false
    Mesh Sampling Type=::NewEnumerator6
    Skeletal Mesh=$System.SkeletalMesh
   ) {module=SkeletalMeshLocation001}
   &SolveForcesAndVelocity(Clamp Velocity=false, Limit Acceleration=false)
  ]
  rendererClasses [&NiagaraMeshRendererProperties]
  renderers [
   {rendererClass=&NiagaraMeshRendererProperties, rendererIndex=0}
   {rendererClass=&NiagaraMeshRendererProperties, rendererIndex=1}
   {rendererClass=&NiagaraMeshRendererProperties, rendererIndex=2}
   {rendererClass=&NiagaraMeshRendererProperties, rendererIndex=3}
  ]
 }
 NE_VC_DriftTrail_WheelWrap_Parent001 {
  emitterUpdateScript [
   &EmitterState(
    Inactive Response=::NewEnumerator1
    Life Cycle Mode=::NewEnumerator1
    Loop Behavior=::NewEnumerator1
    Loop Duration Mode=::NewEnumerator1
    UseLoopDelay=false
   )
   &SpawnBurst_Instantaneous(Spawn Count=4, Use Spawn Probability=false, UseLoopCountLimit=false)
  ]
  particleSpawnScript [
   &VC_PaintedVariantsOverride(
    Black=<rgba 0.08 0.08 0.08 5>
    BurntSienna=<rgba 0.6 0.2728 0.1684 1>
    Cobalt=<rgba 0.2717 0.4348 1.25 1>
    Crimson=<rgba 1.5 0.105 0.1428 2>
    DefaultColor=<rgba 0.7043 0.8338 1 0.5>
    ForestGreen=<rgba 0.5143 0.7899 0.5282 1>
    Grey=<rgba 0.35 0.35 0.35 1>
    Lime=<rgba 0.7328 1 0.1094 1>
    Orange=<rgba 0.89 0.3104 0 1>
    Pink=<rgba 0.8837 0.3977 0.7069 1>
    Purple=<rgba 0.2873 0 0.7 1>
    R1 Black:LinearColor=<rgba 0.15 0.15 0.15 1>
    R1 BurntSienna:LinearColor=<rgba 0.4566 0.2223 0.1264 1>
    R1 Cobalt:LinearColor=<rgba 0.3 0.4167 1 1>
    R1 Crimson:LinearColor=<rgba 1 0.3 0.3 0.7>
    R1 DefaultColor=<rgba 0.24 0.66 0.8 1>
    R1 ForestGreen:LinearColor=<rgba 0 0.5816 0.0294 1>
    R1 Grey:LinearColor=<rgba 0.6 0.6 0.6 0.5>
    R1 Lime:LinearColor=<rgba 0.9312 1.25 0.1875 0.3>
    R1 Orange:LinearColor=<rgba 0.8 0.3218 0.0656 1>
    R1 Pink:LinearColor=<rgba 0.7965 0.3172 0.7824 1>
    R1 Purple:LinearColor=<rgba 0.34 0.17 0.68 0.8>
    R1 Saffron:LinearColor=<rgba 0.9 0.7635 0.315 0.5>
    R1 SkyBlue:LinearColor=<rgba 0.3354 0.6702 0.8385 0.8>
    R1 TitaniumWhite:LinearColor=<rgba 0.6441 0.6441 0.6441 0.8>
    R2 Black:LinearColor=<rgba 0.02 0.02 0.02 10>
    R2 BurntSienna:LinearColor=<rgba 0.1302 0.0592 0.0319 1>
    R2 Cobalt:LinearColor=<rgba 0.05 0.0875 0.5 1>
    R2 Crimson:LinearColor=<rgba 0.3 0 0.0081 1>
    R2 DefaultColor:LinearColor=<rgba 0.3067 0.2 1 0.5>
    R2 ForestGreen:LinearColor=<rgba 0 0.2274 0.0115 1>
    R2 Grey:LinearColor=<rgba 0.15 0.15 0.15 1>
    R2 Lime:LinearColor=<rgba 0.0583 0.1 0 2>
    R2 Orange:LinearColor=<rgba 0.2448 0.0204 0 2>
    R2 Pink:LinearColor=<rgba 0.1913 0 0.0957 0.5965>
    R2 Purple:LinearColor=<rgba 0.0664 0 0.2656 2>
    R2 Saffron:LinearColor=<rgba 0.1788 0.1341 0 1>
    R2 SkyBlue:LinearColor=<rgba 0 0.1934 0.3316 1>
    R2 TitaniumWhite:LinearColor=<rgba 0.2 0.2 0.2 0.8>
    Saffron=<rgba 0.6684 0.5124 0 1>
    SkyBlue=<rgba 0.5 0.8327 1 1>
    TitaniumWhite=<rgba 0.6 0.6 0.6 1>
    Use R1 Painted Overrides=true
    Use R2 Painted Overrides=true
   )
   &InitializeParticle(
    Color=$Particles.Color
    Mesh Scale Mode=::NewEnumerator4
    Sprite Rotation Mode=::NewEnumerator2
    Sprite Size Mode=::NewEnumerator3
    UsePositionOffset=false
    Write Material Random=true
   )
   &PartitionParticles()
   &SkeletalMeshLocation(
    Bone / Socket Index=&Add_Integer(A=$Particles.PartitionParticles.Partition, B=4)
    Bone / Socket Sampling Mode=::NewEnumerator5
    Interpolate Mesh Sampling=false
    MaskSpawnGroup=false
    Mesh Sampling Type=::NewEnumerator6
    Sampled Position Offset=&AddVectorToPosition(Position=<0 0 0>, Vector=$User.offset)
    Skeletal Mesh=$System.SkeletalMesh
   )
  ]
  particleUpdateScript [
   &FN_DisableTireFX(
    Mask Index Attribute:NiagaraInt32=$Particles.PartitionParticles.Partition
    Mask Index FL=2
    Mask Index FR=3
    Mask Index RL=0
    Mask Index RR=1
   )
   &VC_WheelWrapScale(StandardRadius=49, Wheel Index Attribute:NiagaraInt32=&Add_Integer(A=$Particles.PartitionParticles.Partition, B=2))
   &ParticleState(Kill Particles When Lifetime Has Elapsed=false)
   &UpdateMeshOrientation(Rotation Coordinate Space=::NewEnumerator2, Rotation Vector=<0 1 0>) {enabled=false}
   &SkeletalMeshLocation(
    Bone / Socket Index=&Add_Integer(A=$Particles.PartitionParticles.Partition, B=4)
    Bone / Socket Sampling Mode=::NewEnumerator5
    MaskSpawnGroup=false
    Mesh Sampling Type=::NewEnumerator6
    Skeletal Mesh=$System.SkeletalMesh
   ) {module=SkeletalMeshLocation001}
   &SolveForcesAndVelocity(Clamp Velocity=false, Limit Acceleration=false)
  ]
  rendererClasses [&NiagaraMeshRendererProperties]
  renderers [
   {rendererClass=&NiagaraMeshRendererProperties, rendererIndex=0}
   {rendererClass=&NiagaraMeshRendererProperties, rendererIndex=1}
   {rendererClass=&NiagaraMeshRendererProperties, rendererIndex=2}
   {rendererClass=&NiagaraMeshRendererProperties, rendererIndex=3}
  ]
 }
}
summary {
 systemName=NS_DriftTrail_Hitodama_MP
 emitters [
  {bEnabled=false, emitterName=NE_VC_DriftTrail_WheelWrap_Parent, rendererClasses [&NiagaraMeshRendererProperties]}
  {bEnabled=false, emitterName=NE_VC_DriftTrail_Sprite_Parent, rendererClasses [&NiagaraSpriteRendererProperties]}
  {bEnabled=false, emitterName=NE_VC_DriftTrail_Ribbon_Parent, rendererClasses [&NiagaraRibbonRendererProperties]}
  {bEnabled=false, emitterName=NE_VC_DriftTrail_Sprite_Parent002, rendererClasses [&NiagaraSpriteRendererProperties]}
  {bEnabled=false, emitterName=FillTrail, rendererClasses [&NiagaraRibbonRendererProperties]}
  {emitterName=NE_VC_DriftTrail_WheelWrap_Parent001, rendererClasses [&NiagaraMeshRendererProperties]}
  {emitterName=NE_VC_DriftTrail_Sprite_Parent001, rendererClasses [&NiagaraSpriteRendererProperties]}
  {emitterName=NE_VC_DriftTrail_Ribbon_Parent001, rendererClasses [&NiagaraRibbonRendererProperties]}
  {emitterName=NE_VC_DriftTrail_Sprite_Parent003, rendererClasses [&NiagaraSpriteRendererProperties]}
  {emitterName=FillTrail001, rendererClasses [&NiagaraRibbonRendererProperties]}
 ]
 userVariables [
  User.bIsFrontendPreview {type=NiagaraBool}
  User.bIsPainted {type=NiagaraBool}
  User.ColorVariantID {type=NiagaraInt32}
  User.offset {type=Vector3f, description="offset from wheels(set through cosmetic template)", defaultValue {struct=Vector3f, value=<0 0 0>}}
  User.PaintedColor {type=LinearColor, description="", defaultValue {struct=LinearColor, value=<rgba 1 1 1 1>}}
  User.PaintedScalar {type=NiagaraFloat, defaultValue=1, description=""}
  User.VehicleFlags {type=NiagaraInt32, defaultValue=806289408}
  User.Wheel_Radius_Array {
   type=NiagaraDataInterfaceArrayFloat
   defaultValue {
    struct=NiagaraExt_VariableValue_DataInterface
    value {
     dataInterface=&self:NiagaraDataInterfaceArrayFloat_1
     dataInterfaceClass=&NiagaraDataInterfaceArrayFloat
    }
   }
  }
  User.Wheel_Width_Array {
   type=NiagaraDataInterfaceArrayFloat
   defaultValue {
    struct=NiagaraExt_VariableValue_DataInterface
    value {
     dataInterface=&self:NiagaraDataInterfaceArrayFloat_0
     dataInterfaceClass=&NiagaraDataInterfaceArrayFloat
    }
   }
  }
  User.WheelTouchingArray {
   type=NiagaraDataInterfaceArrayFloat
   defaultValue {
    struct=NiagaraExt_VariableValue_DataInterface
    value {
     dataInterface=&self:NiagaraDataInterfaceArrayFloat_2
     dataInterfaceClass=&NiagaraDataInterfaceArrayFloat
    }
   }
  }
 ]
}
