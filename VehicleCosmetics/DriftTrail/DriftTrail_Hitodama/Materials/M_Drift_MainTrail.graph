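# Unlit translucent material for the main trail of the Hitodama drift trail cosmetic, drawing two animated wavy lines.
# Sine and cosine waves along U, scrolled by time, displace two offset V-axis sine bands, which are summed, faded by a UV sine mask and multiplied by particle alpha.
# Particle color times Glow feeds an emissive adjustment function.
# A dynamic parameter controls wave frequency (Width), line sharpness, glow and scroll speed.
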Material /VehicleCosmetics/DriftTrail/DriftTrail_Hitodama/Materials/M_Drift_MainTrail
graph {
 properties {
  bAutomaticallySetUsageInEditor=false
  bUsedWithInstancedStaticMeshes=true
  bUsedWithMeshDeformer=true
  bUsedWithNiagaraMeshParticles=true
  bUsedWithNiagaraRibbons=true
  bUsedWithNiagaraSprites=true
  bUsedWithSkeletalMesh=true
  bUsedWithSplineMeshes=true
  blendMode=BLEND_Translucent
  shadingModel=MSM_Unlit
  usedShadingModels=Unlit
 }
 nodes [
  MaterialExpressionVertexColor_5 = ParticleColor()
  n1 = TextureCoordinate()
  n2 = DynamicParameter() {defaultValue=<rgba 1 1 1 -1>, paramNames [Width,Line,Glow,Time]}
  n3 = Multiply(A=ComponentMask(None=@n1) {r=true}, B=@n2.Width)
  n4 = Multiply(A=Time(), B=@n2.Time)
  n5 = Constant() {r=0.2}
  n6 = Multiply(A=Cosine(None=Add(A=@n3, B=@n4)), B=@n5)
  n7 = ComponentMask(None=Add(A=Add(A=@n1, B=@n6), B=Constant2Vector() {g=-0.2})) {g=true}
  n8 = Multiply(A=ComponentMask(None=@n1) {r=true}, B=@n2.Width)
  n9 = Multiply(A=Sine(None=Add(A=@n8, B=@n4)), B=@n5)
  n10 = ComponentMask(None=Add(A=Add(A=@n1, B=@n9), B=Constant2Vector() {g=0.2})) {g=true}
  n11 = Add(A=Power(Base=Sine(None=@n7) {period=2}, Exp=@n2.Line), B=Power(Base=Sine(None=@n10) {period=2}, Exp=@n2.Line))
  n12 = TextureCoordinate()
  n13 = ComponentMask(None=@n12) {r=true}
  n14 = VectorParameter() {defaultValue=<rgba 1 1 0 0>}
  n15 = ComponentMask(None=@n12) {g=true}
  n16 = Multiply(A=Power(Base=Sine(None=@n13) {period=2}, Exp=@n14.R), B=Power(Base=Sine(None=@n15) {period=2}, Exp=@n14.G))
  MaterialExpressionConstantClamp_4 = Saturate(None=Multiply(A=@MaterialExpressionVertexColor_5.A, B=Multiply(A=@n11, B=@n16)))
  LinearInterpolate()
  MaterialFunctionCall(Emissive=Multiply(A=@MaterialExpressionVertexColor_5.RGB, B=@n2.Glow)) {materialFunction=&MF_FX_EmissiveAdjustments}
 ]
}
